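# MaskMaker intro spawn Niagara effect from the Daigo FX set, building a pink-magenta swirl that ends in a large flash.
# Rotating, expanding meshes form the swirl: a two-renderer Center with separate dark color and size, plus SmokeRing, whose spawn rate ramps up.
# Sprite smoke bursts outward at the base, slowed by drag, with reddish energy color fed through dynamic material parameters.
# BigFlash and BigGlow fire delayed, camera-offset sprite bursts with brightness spikes.
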NiagaraSystem /PistolCanary/Assets/FX/Daigo/IntroSpawn/NS_MaskMaker_IntroSpawn
dependencies {
 usedDataInterfaces [&NiagaraDataInterfaceCurve, &NiagaraDataInterfaceColorCurve]
 usedDynamicInputs [&Divide_Float]
 usedModules [
  &self:SystemSpawnScript.SystemScriptSource.SystemScriptGraph.NiagaraNodeAssignment_0.SetVariables_#1
  &SystemState
  &self:SystemSpawnScript.SystemScriptSource.SystemScriptGraph.NiagaraNodeAssignment_1.SetVariables_#2
  &self:MANAGE_ShouldSpawnThisFrame
  &self:MANAGE_UpdateCounter
 ]
}
emitters {
 BigFlash {
  emitterUpdateScript [
   &EmitterState(
    Enable Distance Culling=true
    Min Distance Response=::SleepAndClearParticles
    MinDistance=3000
    Module.bUseMaxDistance=false
    Module.bUseMinDistance=true
    Scalability Mode=::NewEnumerator1
   )
   &SpawnBurst_Instantaneous(Spawn Time=1.8, Use Spawn Probability=false, UseLoopCountLimit=false)
   &SpawnRate(SpawnRate=2, Use Spawn Probability=false) {enabled=false}
  ]
  particleSpawnScript [&InitializeParticle(
    Color=<rgba 0.4896 0.0405 0.2533 1>
    Lifetime=0.7
    Position=<0 0 300>
    Position Mode=::NewEnumerator1
    Sprite Size=<5000 4000>
    Sprite Size Mode=::NewEnumerator4
    UsePositionOffset=false
    Write Material Random=true
   )]
  particleUpdateScript [
   &ParticleState()
   &ScaleSpriteSize(Uniform Curve Sprite Scale=DataInterface() {propertyValues {Curve {keys [
        {interpMode=RCIM_Cubic, leaveTangent=13.2753, tangentMode=RCTM_Break, time=0, value=0}
        {interpMode=RCIM_Cubic, leaveTangent=-0.1507, tangentMode=RCTM_Break, time=0.1, value=1}
        {arriveTangent=-1.7577, interpMode=RCIM_Cubic, tangentMode=RCTM_Break, time=1, value=0}
       ]}}})
   &ScaleColor(
    Scale RGB=&VectorFromFloat(Value=&FloatFromCurve(FloatCurve=DataInterface() {propertyValues {Curve {keys [
          {arriveTangent=35.5243, interpMode=RCIM_Cubic, leaveTangent=35.5243, tangentMode=RCTM_User, time=0, value=1}
          {interpMode=RCIM_Cubic, leaveTangent=-9.6741, tangentMode=RCTM_Break, time=0.15, value=5}
          {interpMode=RCIM_Cubic, tangentMode=RCTM_Break, time=1, value=1}
         ]}}}))
    ScaleA=true
    ScaleRGB=true
   )
   &CameraOffset(Camera Offset Amount=10)
   &ScaleColor(
    Scale RGB=&VectorFromFloat(Value=&FloatFromCurve(FloatCurve=DataInterface() {propertyValues {Curve {keys [
          {interpMode=RCIM_Cubic, leaveTangent=13.2753, tangentMode=RCTM_Break, time=0, value=0}
          {interpMode=RCIM_Cubic, leaveTangent=-0.1507, tangentMode=RCTM_Break, time=0.1, value=5}
          {arriveTangent=-1.7577, interpMode=RCIM_Cubic, tangentMode=RCTM_Break, time=1, value=1}
         ]}}}))
    ScaleA=true
    ScaleRGB=true
   ) {module=ScaleColor001}
  ]
  rendererClasses [&NiagaraSpriteRendererProperties]
  renderers [{rendererClass=&NiagaraSpriteRendererProperties}]
 }
 BigGlow {
  emitterUpdateScript [
   &EmitterState(
    Enable Distance Culling=true
    Min Distance Response=::SleepAndClearParticles
    MinDistance=3000
    Module.bUseMaxDistance=false
    Module.bUseMinDistance=true
    Scalability Mode=::NewEnumerator1
   )
   &SpawnBurst_Instantaneous(Spawn Time=1.8, Use Spawn Probability=false, UseLoopCountLimit=false)
   &SpawnRate(SpawnRate=2, Use Spawn Probability=false) {enabled=false}
  ]
  particleSpawnScript [
   &InitializeParticle(
    Color=<rgba 1 0.1311 0.6652 0.1>
    Lifetime=0.8
    Position=<0 0 300>
    Position Mode=::NewEnumerator1
    Sprite Size Mode=::NewEnumerator3
    Uniform Sprite Size=12000
    UsePositionOffset=false
    Write Material Random=true
   )
   &DynamicMaterialParameters(
    Index 0 Param 1=0.2
    Index 0 Param 2=1
    Module.Index0Float0Write=true
    Module.Index0Float1Write=true
    Module.Index0Float2Write=true
    Module.Index0Float3Write=true
   )
  ]
  particleUpdateScript [
   &ParticleState()
   &ScaleSpriteSize(Uniform Curve Sprite Scale=DataInterface() {propertyValues {Curve {keys [
        {interpMode=RCIM_Cubic, leaveTangent=13.2753, tangentMode=RCTM_Break, time=0, value=0}
        {interpMode=RCIM_Cubic, leaveTangent=-0.1507, tangentMode=RCTM_Break, time=0.1, value=1}
        {arriveTangent=-1.7577, interpMode=RCIM_Cubic, tangentMode=RCTM_Break, time=1, value=0}
       ]}}})
   &ScaleColor(
    Scale RGB=&VectorFromFloat(Value=&FloatFromCurve(FloatCurve=DataInterface() {propertyValues {Curve {keys [
          {arriveTangent=35.5243, interpMode=RCIM_Cubic, leaveTangent=35.5243, tangentMode=RCTM_User, time=0, value=1}
          {interpMode=RCIM_Cubic, leaveTangent=-9.6741, tangentMode=RCTM_Break, time=0.15, value=5}
          {interpMode=RCIM_Cubic, tangentMode=RCTM_Break, time=1, value=1}
         ]}}}))
    ScaleA=true
    ScaleRGB=true
   )
   &CameraOffset(Camera Offset Amount=80)
  ]
  rendererClasses [&NiagaraSpriteRendererProperties]
  renderers [{rendererClass=&NiagaraSpriteRendererProperties}]
 }
 Center {
  emitterUpdateScript [
   &EmitterState(Life Cycle Mode=::NewEnumerator1, Loop Behavior=::NewEnumerator1, Loop Delay=0.4, Loop Duration=2, UseLoopDelay=true)
   &SpawnBurst_Instantaneous(Use Spawn Probability=true, UseLoopCountLimit=false)
  ]
  particleSpawnScript [
   &InitializeParticle(
    Lifetime=2
    Mesh Scale=<2.4 2.4 3>
    Mesh Scale Mode=::NewEnumerator4
    Position Offset=<0 0 -20>
    UsePositionOffset=true
    Write Material Random=true
   )
   &InitialMeshOrientation(
    Mesh Orientation Mode=::NewEnumerator2
    Rotation=&RandomRangeVector(Minimum=<0 0 0>, Maximum=<0 0 1>)
    Use Rotation Vector=true
   )
   &CameraOffset(Camera Offset Amount=6)
   &ShapeLocation(MaskSpawnGroup=false, Offset=<0 0 125.8756>, Offset Mode=::NewEnumerator0, Override Seed=false, Sphere Radius=1, UseOwnerScale=true)
  ]
  particleUpdateScript [
   &ParticleState()
   &UpdateMeshOrientation(
    Rotation Rate=&Multiply_Float(
     A=-1
     B=&FloatFromCurve(FloatCurve=DataInterface() {propertyValues {Curve {keys [<at 0 1>,<at 0.5 1>,<at 1 0.45>]}}})
    )
    Rotation Vector=<0 0 1>
   )
   &ScaleMeshSize(Scale Factor=&Multiply_Vector(
     A=&MakeVector(
      X=&FloatFromCurve(FloatCurve=DataInterface() {propertyValues {Curve {keys [<at 0 0.5>,<at 1 1.5>]}}})
      Y=&FloatFromCurve(FloatCurve=DataInterface() {propertyValues {Curve {keys [<at 0 0.5>,<at 1 1.5>]}}})
      Z=1
     )
     B=&VectorFromFloat(Value=2)
    ))
   &ScaleColor(
    Scale Alpha=&FloatFromCurve(
     FloatCurve=DataInterface() {propertyValues {Curve {keys [
         {arriveTangent=2.8927, interpMode=RCIM_Cubic, time=0, value=0}
         {arriveTangent=0.047, interpMode=RCIM_Cubic, leaveTangent=0.047, tangentMode=RCTM_User, time=0.5, value=1.325}
         {arriveTangent=-2.3801, interpMode=RCIM_Cubic, leaveTangent=-2.3801, tangentMode=RCTM_User, time=0.7, value=0.95}
         <at 1 0 RCIM_Cubic>
        ]}}}
     Scale Curve=0.5
    )
    Scale RGB=&MakeVectorFromLinearColorRGB(Color=<rgba 1 0.0827 0.5173 1>)
    ScaleA=true
    ScaleRGB=true
   )
   /SetVariables(Particles.Dark_Color:LinearColor=&MakeLinearColorFromVectorAndFloat(
     "Vector (RGB)"=&MakeVectorFromLinearColorRGB(Color=<rgba 0.1562 0.0067 0.0429 1>)
     "Float (Alpha)"=&FloatFromCurve(
      FloatCurve=DataInterface() {propertyValues {Curve {keys [
          {arriveTangent=2.8927, interpMode=RCIM_Cubic, time=0, value=0}
          {arriveTangent=0.047, interpMode=RCIM_Cubic, leaveTangent=0.047, tangentMode=RCTM_User, time=0.4, value=0.9}
          {arriveTangent=-1.5359, interpMode=RCIM_Cubic, leaveTangent=-1.5359, tangentMode=RCTM_User, time=0.8, value=0.35}
          <at 1 0 RCIM_Cubic>
         ]}}}
      Scale Curve=0.5
     )
    )) {isSetParameters=true}
   /SetVariables(Particles.Dark_Size:Vector3f=&Multiply_Vector(
     A=&MakeVector(
      X=&FloatFromCurve(FloatCurve=DataInterface() {propertyValues {Curve {keys [<at 0 0.5>,<at 1 1.3>]}}}, Scale Curve=4.4)
      Y=&FloatFromCurve(FloatCurve=DataInterface() {propertyValues {Curve {keys [<at 0 0.5>,<at 1 1.3>]}}}, Scale Curve=4.4)
      Z=3
     )
     B=&VectorFromFloat(Value=2)
    )) {isSetParameters=true}
   &SolveForcesAndVelocity(Clamp Velocity=false, Limit Acceleration=false)
  ]
  rendererClasses [&NiagaraMeshRendererProperties]
  renderers [
   {rendererClass=&NiagaraMeshRendererProperties, rendererIndex=0}
   {rendererClass=&NiagaraMeshRendererProperties, rendererIndex=1}
  ]
 }
 SmokeRing {
  emitterUpdateScript [
   &EmitterState(Life Cycle Mode=::NewEnumerator1, Loop Behavior=::NewEnumerator1, Loop Duration=1.5, UseLoopDelay=false)
   &SpawnBurst_Instantaneous(Spawn Count=8, Use Spawn Probability=false, UseLoopCountLimit=false) {enabled=false}
   &SpawnRate(
    SpawnRate=&FloatFromCurve(
     FloatCurve=DataInterface() {propertyValues {Curve {defaultValue=1, keys [{tangentMode=RCTM_User, time=0, value=0}, {tangentMode=RCTM_User, time=0.7, value=1}]}}}
     CurveIndex=$Emitter.NormalizedLoopAge
     Scale Curve=20
    )
    Use Spawn Probability=false
   )
  ]
  particleSpawnScript [
   &InitializeParticle(
    Color Maximum=<rgba 0.3 0.3 0.3 1>
    Color Minimum=<rgba 0.2 0.2 0.2 1>
    Color Mode=::NewEnumerator2
    Lifetime Max=1
    Lifetime Min=0.5
    Lifetime Mode=::NewEnumerator1
    Mesh Scale=&Multiply_VectorByFloat(
     Vector=&RandomRangeVector(Minimum=<1 1 1.4>, Maximum=<1 1 1.8>)
     Float=&RandomRangeFloat(Minimum=0.45, Maximum=0.55)
    )
    Mesh Scale Mode=::NewEnumerator4
    UsePositionOffset=false
    Write Material Random=true
   )
   &InitialMeshOrientation(
    Mesh Orientation Mode=::NewEnumerator2
    Rotation=&RandomRangeVector(Minimum=<0.02 0.02 -1>, Maximum=<0.03 0.03 1>)
    Use Rotation Vector=true
   )
   &ShapeLocation(
    Cylinder Height=300
    Cylinder Radius=1
    MaskSpawnGroup=false
    Offset=<0 0 304.1358>
    Offset Mode=::NewEnumerator0
    Override Seed=false
    Shape Primitive=::NewEnumerator1
    Surface Only=false
    UseOwnerScale=true
   )
  ]
  particleUpdateScript [
   &ParticleState()
   &DynamicMaterialParameters(
    Index 0 Param 1=&FloatFromCurve(FloatCurve=DataInterface() {propertyValues {Curve {keys [{interpMode=RCIM_Cubic, tangentMode=RCTM_User, time=0, value=0.8}, {interpMode=RCIM_Cubic, tangentMode=RCTM_User, time=1, value=0}]}}})
    Module.Param0WriteEnabled=true
    Module.Param1WriteEnabled=true
    Module.Param2WriteEnabled=true
    Module.Param3WriteEnabled=true
   )
   &ScaleMeshSize(Scale Factor=&Multiply_VectorByFloat(
     Vector=<1 1 1>
     Float=&FloatFromCurve(
      FloatCurve=DataInterface() {propertyValues {Curve {keys [
          {arriveTangent=0.5864, interpMode=RCIM_Cubic, leaveTangent=0.5864, tangentMode=RCTM_User, time=0, value=0.5}
          {arriveTangent=0.4564, interpMode=RCIM_Cubic, leaveTangent=0.4564, tangentMode=RCTM_User, time=0.5954, value=0.6852}
          {arriveTangent=0.5662, interpMode=RCIM_Cubic, leaveTangent=0.5662, tangentMode=RCTM_User, time=1, value=1.1}
         ]}}}
      Scale Curve=8
     )
    ))
   &ScaleMeshSize(Scale Factor=&MakeVector(
     X=&FloatFromCurve(
      FloatCurve=DataInterface() {propertyValues {Curve {keys [
          {arriveTangent=0.3413, interpMode=RCIM_Cubic, leaveTangent=0.3413, tangentMode=RCTM_User, time=0.8, value=1}
          {arriveTangent=3.5208, interpMode=RCIM_Cubic, leaveTangent=3.5208, tangentMode=RCTM_User, time=1, value=1.5}
         ]}}}
      CurveIndex=$Emitter.Age
     )
     Y=&FloatFromCurve(
      FloatCurve=DataInterface() {propertyValues {Curve {keys [
          {arriveTangent=0.3598, interpMode=RCIM_Cubic, leaveTangent=0.3598, tangentMode=RCTM_User, time=0.8, value=1}
          {arriveTangent=3.5574, interpMode=RCIM_Cubic, leaveTangent=3.5574, tangentMode=RCTM_User, time=1, value=1.5}
         ]}}}
      CurveIndex=$Emitter.Age
     )
     Z=1
    )) {enabled=false, module=ScaleMeshSize001}
   &UpdateMeshOrientation(
    Rotation Rate=&Multiply_Float(
     A=-1
     B=&FloatFromCurve(FloatCurve=DataInterface() {propertyValues {Curve {keys [<at 0 1>,<at 0.4 1>,<at 1 0>]}}})
    )
    Rotation Vector=<0 0 1>
   )
   &ScaleColor(
    Scale Alpha=&FloatFromCurve(
     FloatCurve=DataInterface() {propertyValues {Curve {keys [
         {arriveTangent=4.9076, interpMode=RCIM_Cubic, leaveTangent=4.9076, tangentMode=RCTM_User, time=0, value=0}
         {arriveTangent=0.0332, interpMode=RCIM_Cubic, leaveTangent=0.0332, tangentMode=RCTM_User, time=0.1, value=1}
         {arriveTangent=-0.0163, interpMode=RCIM_Cubic, leaveTangent=-0.025, tangentMode=RCTM_Break, time=0.4, value=1}
         {arriveTangent=-1.5, interpMode=RCIM_Cubic, leaveTangent=-1.5, time=0.6, value=0.8}
         {arriveTangent=-2, interpMode=RCIM_Cubic, leaveTangent=-2, time=0.9, value=0.25}
         {arriveTangent=-3.0089, tangentMode=RCTM_User, time=1, value=0}
        ]}}}
     Scale Curve=&FloatFromCurve(
      FloatCurve=DataInterface() {propertyValues {Curve {
         defaultValue=1
         keys [<at 0 0 RCIM_Cubic>, {arriveTangent=2.408, interpMode=RCIM_Cubic, leaveTangent=2.408, tangentMode=RCTM_User, time=1, value=1}]
        }}}
      CurveIndex=$System.Age
      Scale Curve=2
     )
    )
    Scale RGB=&MakeVectorFromLinearColorRGB(Color=&ColorFromCurve(ColorCurve=DataInterface() {propertyValues {
        AlphaCurve {keys []}
        BlueCurve {keys [
          <at 0.0254 0.0736>
          {arriveTangent=0.0742, interpMode=RCIM_Cubic, leaveTangent=0.0742, time=0.2471, value=0.5161}
          <at 0.6621 0.1209 RCIM_Cubic>
         ]}
        GreenCurve {keys [
          <at 0.0254 0.0146>
          {arriveTangent=-0.0073, interpMode=RCIM_Cubic, leaveTangent=-0.0073, time=0.2471, value=0.1037}
          <at 0.6621 0.0099 RCIM_Cubic>
         ]}
        RedCurve {keys [<at 0.0254 0.5>, {arriveTangent=-0.3927, interpMode=RCIM_Cubic, leaveTangent=-0.3927, time=0.2471, value=1.5}, <at 0.6621 0.25 RCIM_Cubic>]}
       }}))
    ScaleA=true
    ScaleRGB=true
   )
   &SolveForcesAndVelocity(Clamp Velocity=false, Limit Acceleration=false)
  ]
  rendererClasses [&NiagaraMeshRendererProperties]
  renderers [{rendererClass=&NiagaraMeshRendererProperties, rendererIndex=0}]
 }
 Smoke_Ring001 {
  emitterUpdateScript [
   &EmitterState(
    Enable Distance Culling=true
    Min Distance Response=::SleepAndLetParticlesFinish
    MinDistance=3000
    Module.bUseMaxDistance=false
    Module.bUseMinDistance=true
    Scalability Mode=::NewEnumerator1
   )
   &SpawnRate(SpawnRate=20, Use Spawn Probability=false)
   &SpawnBurst_Instantaneous(Spawn Count=40, Use Spawn Probability=false, UseLoopCountLimit=false) {enabled=false}
   /SetVariables(Emitter.LightingCenter:Vector4f=&MakeFloat4FromVectorAndFloat(
     "Vector (RGB)"=&AddVectorToPosition(Position=$Engine.Emitter.SimulationPosition, Vector=<0 0 0>)
     "Float (Alpha)"=1
    )) {isSetParameters=true}
  ]
  particleSpawnScript [
   &InitializeParticle(
    Color=<rgba 0.7446 0.7604 0.7034 1>
    Lifetime=&RandomRangeFloat(Minimum=1.5, Maximum=2)
    Sprite Rotation Mode=::NewEnumerator2
    Sprite Size Mode=::NewEnumerator3
    Uniform Sprite Size=&Multiply_Float(
     A=&Multiply_Float(A=400, B=&Lerp_Float(A=0.5, B=1, Alpha=0.5))
     B=&RandomRangeFloat(Minimum=0.8, Maximum=1)
    )
    UsePositionOffset=true
    Write Material Random=true
   ) {module=InitializeParticle001}
   &CameraOffset(Camera Offset Amount=&Multiply_Float(
     A=&RandomRangeFloat(Minimum=25, Maximum=50)
     B=&SelectFloatFromArray(Direct Array Index=$Particles.PartitionParticles.Partition, Float Selection Array=DataInterface() {propertyValues {FloatData [0,0,0]}})
    ))
   &ShapeLocation(
    Cylinder Height=10
    Cylinder Radius=300
    MaskSpawnGroup=false
    Override Seed=false
    Rotation Axis=<0 0 1>
    Rotation Mode=::NewEnumerator1
    Shape Primitive=::NewEnumerator1
    Surface Only=false
    UseOwnerScale=true
   )
   &AddVelocity(
    Module.Use Falloff Distance=false
    Override Seed=false
    Velocity Mode=::NewEnumerator1
    Velocity Speed=&Multiply_Float(
     A=&RandomRangeFloat(Minimum=0.2, Maximum=1)
     B=&Multiply_Float(A=2000, B=0.3)
    )
   )
  ]
  particleUpdateScript [
   &ParticleState()
   &ScaleSpriteSize(Scale Factor=&Vector2DFromFloat(Value=&FloatFromCurve(
      FloatCurve=DataInterface() {propertyValues {Curve {keys [{interpMode=RCIM_Cubic, leaveTangent=1.1499, tangentMode=RCTM_Break, time=0, value=0.5}, <at 0.8 1>]}}}
      Scale Curve=1.3
     )))
   &Drag(Drag=2, Use Linear Drag=true, Use Rotational Drag=false)
   &SolveForcesAndVelocity(Clamp Velocity=false, Limit Acceleration=false)
   /SetVariables(Particles.Color_ParticleEnergy:LinearColor=&ColorFromCurve(ColorCurve=DataInterface() {propertyValues {
       AlphaCurve {keys [<at 0 1 RCIM_Cubic>]}
       BlueCurve {keys [<at 0.3364 0.3332>,<at 0.749 0.2426 RCIM_Cubic>]}
       GreenCurve {keys [<at 0.3364 0.3531>,<at 0.749 0.3089 RCIM_Cubic>]}
       RedCurve {keys [<at 0.3364 2>,<at 0.749 5 RCIM_Cubic>]}
      }})) {isSetParameters=true}
   &DynamicMaterialParameters(
    Index 0 Param 1=&FloatFromCurve(FloatCurve=DataInterface() {propertyValues {Curve {keys [{arriveTangent=2.0446, interpMode=RCIM_Cubic, leaveTangent=2.0446, tangentMode=RCTM_User, time=0, value=0}, <at 1 1 RCIM_Cubic>]}}})
    Index 0 Param 3=&FloatFromCurve(FloatCurve=DataInterface() {propertyValues {Curve {keys [
         {interpMode=RCIM_Cubic, tangentMode=RCTM_Break, time=0, value=0}
         {interpMode=RCIM_Cubic, leaveTangent=-3.8707, tangentMode=RCTM_Break, time=0.2, value=1}
         {tangentMode=RCTM_User, time=1, value=0}
        ]}}})
    Index 0 Param 4=&Multiply_Float(
     A=&FloatFromCurve(
      FloatCurve=DataInterface() {propertyValues {Curve {
         defaultValue=1
         keys [
          {interpMode=RCIM_Cubic, tangentMode=RCTM_User, time=0, value=0}
          {interpMode=RCIM_Cubic, tangentMode=RCTM_User, time=0.5, value=1}
          {tangentMode=RCTM_User, time=1, value=0}
         ]
        }}}
      Scale Curve=15
     )
     B=&FloatFromCurve(
      FloatCurve=DataInterface() {propertyValues {Curve {keys [<at 0 1 RCIM_Cubic>, {arriveTangent=-0.4, interpMode=RCIM_Cubic, leaveTangent=-0.4, time=1, value=0.75}, <at 2.5 0 RCIM_Cubic>]}}}
      CurveIndex=$Emitter.Age
     )
    )
    Index 1 Param 1=&MakeFloatFromVector(Vector=&MakeVectorFromLinearColorRGB(Color=$Particles.Color_ParticleEnergy), Channel=::NewEnumerator0)
    Index 1 Param 2=&MakeFloatFromVector(Vector=&MakeVectorFromLinearColorRGB(Color=$Particles.Color_ParticleEnergy), Channel=::NewEnumerator1)
    Index 1 Param 3=&MakeFloatFromVector(Vector=&MakeVectorFromLinearColorRGB(Color=$Particles.Color_ParticleEnergy), Channel=::NewEnumerator2)
    Index 1 Param 4=&FloatFromCurve(FloatCurve=DataInterface() {propertyValues {Curve {keys [{interpMode=RCIM_Cubic, tangentMode=RCTM_Break, time=0.4, value=0}, {arriveTangent=3.131, tangentMode=RCTM_User, time=1, value=1}]}}})
    Index 2 Param 1=500
    Index 2 Param 2=500
    Index 2 Param 3=500
    Module.Index0Param0Write=true
    Module.Index0Param1Write=true
    Module.Index0Param2Write=true
    Module.Index0Param3Write=true
    Module.Index1Param0Write=true
    Module.Index1Param1Write=true
    Module.Index1Param2Write=true
    Module.Index1Param3Write=true
    Module.Index2Param0Write=true
    Module.Index2Param1Write=true
    Module.Index2Param2Write=true
    Module.Index2Param3Write=true
    Write Parameter Index 1=true
    Write Parameter Index 2=true
   )
   &ScaleColor(
    Scale Alpha=&FloatFromCurve(
     FloatCurve=DataInterface() {propertyValues {Curve {keys [
         {interpMode=RCIM_Cubic, leaveTangent=7.3008, tangentMode=RCTM_Break, time=0, value=0}
         {interpMode=RCIM_Cubic, tangentMode=RCTM_Break, time=0.2, value=1}
         {arriveTangent=0.0179, interpMode=RCIM_Cubic, leaveTangent=0.0179, tangentMode=RCTM_User, time=1, value=0}
        ]}}}
     Scale Curve=0.8
    )
    ScaleA=true
    ScaleRGB=true
   )
  ]
  rendererClasses [&NiagaraSpriteRendererProperties]
  renderers [{rendererClass=&NiagaraSpriteRendererProperties}]
 }
}
summary {
 systemName=NS_MaskMaker_IntroSpawn
 emitters [
  {emitterName=Smoke_Ring001, rendererClasses [&NiagaraSpriteRendererProperties]}
  {emitterName=Center, rendererClasses [&NiagaraMeshRendererProperties]}
  {emitterName=SmokeRing, rendererClasses [&NiagaraMeshRendererProperties]}
  {emitterName=BigFlash, rendererClasses [&NiagaraSpriteRendererProperties]}
  {emitterName=BigGlow, rendererClasses [&NiagaraSpriteRendererProperties]}
 ]
 userVariables [
  User.bIsFrontEnd {type=NiagaraBool}
  User.bIsFrontEndPreview {type=NiagaraBool}
  User.FXHeight {type=NiagaraFloat, defaultValue=7, description=""}
  User.FXRadius {type=NiagaraFloat, defaultValue=3, description=""}
  User.RandomArray {
   type=NiagaraDataInterfaceArrayFloat
   defaultValue {
    struct=NiagaraExt_VariableValue_DataInterface
    value {
     dataInterface=&self:NiagaraDataInterfaceArrayFloat_0
     dataInterfaceClass=&NiagaraDataInterfaceArrayFloat
    }
   }
  }
  User.SpawnArray {
   type=NiagaraDataInterfaceArrayPosition
   description=""
   defaultValue {
    struct=NiagaraExt_VariableValue_DataInterface
    value {
     dataInterface=&self:NiagaraDataInterfaceArrayPosition_0
     dataInterfaceClass=&NiagaraDataInterfaceArrayPosition
    }
   }
  }
  User.SpawnCounter {
   type=NiagaraDataInterfaceSimpleCounter
   description=""
   defaultValue {
    struct=NiagaraExt_VariableValue_DataInterface
    value {
     dataInterface=&self:NiagaraDataInterfaceSimpleCounter_0
     dataInterfaceClass=/Script/Niagara.NiagaraDataInterfaceSimpleCounter
    }
   }
  }
 ]
}
